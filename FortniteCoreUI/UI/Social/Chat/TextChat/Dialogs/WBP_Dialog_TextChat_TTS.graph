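# Accessibility popup dialog that lets players switch text chat text-to-speech and speech-to-text on or off, then close with Done.
# Construct creates one FortSocialSettingDiscreteVM, which is retargeted by setting ID to read both values and again on row hover or focus.
# Clicks and left/right navigation call ToggleOption, which updates the rotator material and On/Off label and writes to whichever setting the shared VM currently targets, including during Construct.
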WidgetBlueprint /FortniteCoreUI/UI/Social/Chat/TextChat/Dialogs/WBP_Dialog_TextChat_TTS
graphs [
 BP_GetDesiredFocusTarget {nodes [
   Get Desired Focus Target()
   > Return Node!(execute, ReturnValue:Widget Object Reference=Get Setting_TTS() -> (Setting_TTS:WBP Text Chat Entry Message Base Object Reference))
  ]}
 EventGraph {nodes [
   "On Clicked (Setting_STT)"!() -> (OutputDelegate, then, Button:Common Button Base Object Reference) {is=title K2Node_ComponentBoundEvent}
   > n1 = Set isSpeechTT Enabled(isSpeechTT Enabled:bool=NOT Boolean(A=Get isSpeechTT Enabled() -> (isSpeechTT Enabled:bool))) -> (Output_Get:bool)
   > ToggleOption!(
    execute
    self
    isEnabled:bool=@n1
    Border:Border Object Reference=Get STT_Enabled() -> (STT_Enabled:Border Object Reference)
    Common Text:Common Text Object Reference=Get STT_OnOff() -> (STT_OnOff:Common Text Object Reference)
   ) -> (then)
   > Branch(Condition=IsUsingCompactUI())
   > SetVisibility(
    self=Get WBP_UIKit_Block_Outline!(self:WBP Text Chat Entry Message Base Object Reference=Get Setting_TTS()) -> (WBP_UIKit_Block_Outline:WBP UIKit Block Outline Object Reference)
    InVisibility=Collapsed
   )
   > SetVisibility(
    self=Get WBP_UIKit_Block_Outline(self=Get Setting_STT() -> (Setting_STT:WBP Text Chat Entry Message Base Object Reference))
    InVisibility=SelfHitTestInvisible
   )
   "On Clicked (Setting_TTS)"!() -> (OutputDelegate, then, Button:Common Button Base Object Reference) {is=title K2Node_ComponentBoundEvent}
   > n2 = Set is Text TS Enabled(is Text TS Enabled:bool=NOT Boolean(A=Get is Text TS Enabled() -> (is Text TS Enabled:bool))) -> (Output_Get:bool)
   > ToggleOption(
    isEnabled=@n2
    Border=Get TTS_Enabled() -> (TTS_Enabled:Border Object Reference)
    Common Text=Get TTS_OnOff() -> (TTS_OnOff:Common Text Object Reference)
   )
   > Branch(Condition=IsUsingCompactUI())
   > SetVisibility(self=Get WBP_UIKit_Block_Outline(self=Get Setting_STT()), InVisibility=Collapsed)
   > SetVisibility(self=Get WBP_UIKit_Block_Outline(self=Get Setting_TTS()), InVisibility=SelfHitTestInvisible)
   n3 = "On Hovered (Setting_TTS)"!() -> (OutputDelegate, then, Button:Common Button Base Object Reference) {is=title K2Node_ComponentBoundEvent}
   n4 = Get FortSocialSettingDiscreteVM() -> (FortSocialSettingDiscreteVM:Fort Social Setting Discrete VM Object Reference)
   SetSettingId!(execute=[@n3.then, @n5.then], self:Fort Setting Value VM Object Reference=@n4, InSettingId:String=ShowTextToSpeech) -> (then)
   n5 = "On Focused (Setting_TTS)"!() -> (OutputDelegate, then, Button:Common Button Base Object Reference) {is=title K2Node_ComponentBoundEvent}
   n6 = "On Hovered (Setting_STT)"!() -> (OutputDelegate, then, Button:Common Button Base Object Reference) {is=title K2Node_ComponentBoundEvent}
   SetSettingId(execute=[@n6.then, @n7.then], self=@n4, InSettingId=ShowSpeechToText)
   n7 = "On Focused (Setting_STT)"!() -> (OutputDelegate, then, Button:Common Button Base Object Reference) {is=title K2Node_ComponentBoundEvent}
   "On Clicked (WBP_UIKit_Button_Regular)"!() -> (OutputDelegate, then, Button:Common Button Base Object Reference) {is=title K2Node_ComponentBoundEvent}
   > DeactivateWidget()
   Event PreConstruct()
   > n8 = GetDynamicMaterial(self:Border Object Reference=Get TTS_Enabled())
   > SetScalarParameterValue(self=@n8, ParameterName=Options, Value=2)
   > n9 = GetDynamicMaterial(self=Get STT_Enabled())
   > SetScalarParameterValue(self=@n9, ParameterName=Options, Value=2)
   Event Construct()
   > n10 = Construct Object from Class(Class=/Script/FortniteUI.FortSocialSettingDiscreteVM) -> (ReturnValue:Fort Social Setting Discrete VM Object Reference)
   > n11 = Set FortSocialSettingDiscreteVM(FortSocialSettingDiscreteVM:Fort Social Setting Discrete VM Object Reference=@n10) -> (Output_Get:Fort Social Setting Discrete VM Object Reference)
   > SetSettingId(self=@n11, InSettingId=ShowSpeechToText)
   n12 = Get FortSocialSettingDiscreteVM()
   > Set isSpeechTT Enabled(isSpeechTT Enabled=Select(
     Option 0:bool=false
     Option 1:bool=true
     Index:int=GetSettingValueIndex!(self:Fort Social Setting Discrete VM Object Reference=@n12) -> (ReturnValue:int=0)
    ) -> (ReturnValue:bool))
   > SetSettingId(self=Get FortSocialSettingDiscreteVM(), InSettingId=ShowTextToSpeech)
   n13 = Get FortSocialSettingDiscreteVM()
   > Set is Text TS Enabled(is Text TS Enabled=Select(Option 0=false, Option 1=true, Index=GetSettingValueIndex(self=@n13) -> (ReturnValue=0)))
   > ToggleOption(isEnabled=Get is Text TS Enabled(), Border=Get TTS_Enabled(), Common Text=Get TTS_OnOff())
   > ToggleOption(isEnabled=Get isSpeechTT Enabled(), Border=Get STT_Enabled(), Common Text=Get STT_OnOff())
  ]}
 NavigateToDone {nodes [
   NavigateToDone!() -> (then, Navigation:EUINavigation Enum) {is=title K2Node_FunctionEntry}
   > Return Node(ReturnValue=Get WBP_UIKit_Button_Regular() -> (WBP_UIKit_Button_Regular:WBP UIKit Button Regular Object Reference))
  ]}
 OnMouseWheel {nodes [
   OnMouseWheel!() -> (then, MyGeometry:Geometry Structure, MouseEvent:"Pointer Event Structure (by ref)") {is=title K2Node_FunctionEntry}
   > n1 = Branch(Condition=HasAnyUserFocus!(self:Widget Object Reference=Self-Reference()) -> (ReturnValue:bool))
   > Return Node(ReturnValue:Event Reply Structure=Unhandled())
   Return Node(execute=@n1.else, ReturnValue=Handled())
  ]}
 OnReportClosed {nodes [OnReportClosed!() -> (then, ReportState:ETextReportState Enum=BeginReport) {is=title K2Node_FunctionEntry}]}
 ToggleOption {nodes [
   n1 = ToggleOption!() -> (then, isEnabled:bool=~, Border:Border Object Reference, Common Text:Common Text Object Reference) {is=title K2Node_FunctionEntry}
   > n2 = GetDynamicMaterial(self=Get Border() -> (Border:Border Object Reference))
   > SetScalarParameterValue(self=@n2, ParameterName=SelectedOption, Value="To Float (Boolean)"(InBool=@n1))
   > "SetText (Text)"(
    self=Get Common Text() -> (Common Text:Common Text Object Reference)
    InText=SelectText(A="LOCTEXT(\"On\")", B="LOCTEXT(\"Off\")", bPickA=@n1)
   )
   > SetSettingValueByIndex!(
    execute
    self:Fort Social Setting Discrete VM Object Reference=Get FortSocialSettingDiscreteVM()
    index:int=SelectInt(A=1, B=0, bPickA=Get isEnabled() -> (isEnabled:bool))
   ) -> (then)
  ]}
 ToggleSTTNavigation {nodes [
   n1 = ToggleSTTNavigation!() -> (then, Navigation:EUINavigation Enum) {is=title K2Node_FunctionEntry}
   > n2 = Switch on EUINavigation!(execute, Selection:EUINavigation Enum=@n1) -> (Left:Exec, Right:Exec, Up:Exec, Down:Exec, Next:Exec, Previous:Exec, Invalid:Exec) {is=title K2Node_SwitchEnum}
   n3 = Get isSpeechTT Enabled()
   > n4 = Branch(Condition=@n3)
   ToggleOption(execute=[@n5.else, @n4.then], isEnabled=NOT Boolean(A=Get isSpeechTT Enabled()), Border=Get STT_Enabled(), Common Text=Get STT_OnOff()) {is=title K2Node_CallFunction}
   > Set isSpeechTT Enabled(isSpeechTT Enabled=NOT Boolean(A=Get isSpeechTT Enabled()))
   > Return Node(ReturnValue:Widget Object Reference)
   n5 = Branch(execute=@n2.Right, Condition=@n3)
  ]}
 ToggleTTSNavigation {nodes [
   n1 = ToggleTTSNavigation!() -> (then, Navigation:EUINavigation Enum) {is=title K2Node_FunctionEntry}
   > n2 = Switch on EUINavigation(Selection=@n1)
   > n3 = Branch(Condition=Get is Text TS Enabled())
   ToggleOption(execute=[@n4.else, @n3.then], isEnabled=NOT Boolean(A=Get is Text TS Enabled()), Border=Get TTS_Enabled(), Common Text=Get TTS_OnOff())
   > Set is Text TS Enabled(is Text TS Enabled=NOT Boolean(A=Get is Text TS Enabled()))
   > Return Node()
   n4 = Branch(execute=@n2.Right, Condition=Get is Text TS Enabled())
  ]}
]
variables [
 is Text TS Enabled {type=boolean}
 isSpeechTT Enabled {type=boolean}
 OnReportClosed {}
]
widgets {
 info {
  inheritedWidgetCount=0
  namedSlotCount=3
  parentClass=&CommonActivatableWidget
  rootWidgetClass=/FortUILibrary/Systems/Dialog/WBP_UIKit_Dialog_Internal.WBP_UIKit_Dialog_Internal_C
  widgetCount=27
 }
 widgets [
  Dialog_Internal {
   widgetClassPath=/FortUILibrary/Systems/Dialog/WBP_UIKit_Dialog_Internal
   bIsVariable=true
   widget=Dialog_Internal
   properties {
    bIsFocusable=true
    designTimePreview_DialogVM=null
    titleImageSizeX=64
    titleImageSizeY=64
    titleTextVerticalAligment=VAlign_Center
    useTitle=false
    backplateProperties {
     blur=5
     brightness=Dark
     corner Radius=Backplate
     placement {
      horizontal Alignment=HAlign_Fill
      vertical Alignment=VAlign_Fill
      grid {column=0, column Span=2, layer=0, row=0, row Span=0, nudge=<0 0>}
      padding=<ltrb 0 0 0 0>
     }
    }
    buttons {
     each Button Horizontal Fill=false
     each Button Min Width=258
     each Button Size=Medium
     focuse Button=0
     space Between Buttons=24
     cTA Buttons Types []
     group Placement {
      horizontal Alignment=HAlign_Left
      vertical Alignment=VAlign_Bottom
      grid {column=1, column Span=0, layer=0, row=2, row Span=0, nudge=<0 0>}
      padding=<ltrb 0 0 0 0>
     }
    }
    dialogSIze {
     height Override {enable=false, value=0}
     max Aspect Ratio {enable=false, value=0}
     max Desired Height {enable=true, value=800}
     max Desired Width {enable=false, value=0}
     min Aspect Ratio {enable=false, value=0}
     min Desired Height {enable=false, value=0}
     min Desired Width {enable=true, value=1345}
     width Override {enable=false, value=0}
    }
    titleImagePadding=<ltrb 10 0 15 0>
    titleProperties {
     auto Wrap Text=true
     justification=Left
     line Height Percentage=1
     mobile Font Size Multiplier=1.3
     transform Policy=ToUpper
     wrap Text At=700
     color {colorUseRule=UseColor_Specified, specifiedColor=<rgba 1 1 1 1>}
     font {
      bForceMonospaced=false
      bMaterialIsStencil=false
      fontMaterial=null
      fontObject=&HeadingNow
      letterSpacing=0
      monospacedWidth=1
      size=25.5
      skewAmount=0
      typefaceFontName=86 Bold Italic
      outlineSettings {
       bApplyOutlineToDropShadows=false
       bMiteredCorners=false
       bSeparateFillAlpha=false
       outlineMaterial=null
       outlineSize=0
       outlineColor=<rgba 0 0 0 1>
      }
     }
     placement {
      horizontal Alignment=HAlign_Fill
      vertical Alignment=VAlign_Top
      grid {column=1, column Span=0, layer=0, row=0, row Span=0, nudge=<0 0>}
      padding=<ltrb 80 80 80 32>
     }
    }
   }
  }
  VB_TTSandSTT {widgetClassPath=&VerticalBox, namedSlotHost=Dialog_Internal, widget=VB_TTSandSTT}
  HorizontalBox_1 {
   widgetClassPath=&HorizontalBox
   parent=VB_TTSandSTT
   widget=HorizontalBox_1
   slot {class=VerticalBoxSlot, properties {padding=<ltrb 64 64 64 0>}}
  }
  VerticalBox_1 {widgetClassPath=&VerticalBox, parent=HorizontalBox_1, widget=VerticalBox_1, slot {class=HorizontalBoxSlot}}
  dialog_Title {
   widgetClassPath=&CommonTextBlock
   bIsVariable=true
   parent=VerticalBox_1
   widget=dialog_Title
   properties {
    text="New accessibility Options Available!"
    textTransformPolicy=ToUpper
    wrapTextAt=600
    font {
     bForceMonospaced=false
     bMaterialIsStencil=false
     fontMaterial=null
     fontObject=&HeadingNow
     letterSpacing=10
     monospacedWidth=1
     size=25.5
     skewAmount=0
     typefaceFontName=86 Bold Italic
     outlineSettings {
      bApplyOutlineToDropShadows=false
      bMiteredCorners=false
      bSeparateFillAlpha=false
      outlineMaterial=/FortUILibrary/MaterialInstances/Text/MI_UI_Block_Text_Outline
      outlineSize=2
      outlineColor=<rgba 0 0 0 1>
     }
    }
    shadowOffset=<0 0>
    shapedTextOptions {bOverride_TextFlowDirection=true, bOverride_TextShapingMethod=false, textFlowDirection=Culture, textShapingMethod=Auto}
   }
   slot {class=VerticalBoxSlot, properties {padding=<ltrb 0 0 0 24>}}
  }
  Description {
   widgetClassPath=&CommonTextBlock
   bIsVariable=true
   parent=VerticalBox_1
   widget=Description
   properties {
    text="We’ve added new accessibility options to Fortnite. Change them below or find them later in the Communication Settings tab."
    wrapTextAt=784
    font {
     bForceMonospaced=false
     bMaterialIsStencil=false
     fontMaterial=null
     fontObject=&HeadingNow
     letterSpacing=10
     monospacedWidth=1
     size=18
     skewAmount=0
     typefaceFontName=74 Regular
     outlineSettings {
      bApplyOutlineToDropShadows=false
      bMiteredCorners=false
      bSeparateFillAlpha=false
      outlineMaterial=&MI_UI_Block_Text_Outline
      outlineSize=2
      outlineColor=<rgba 0 0 0 1>
     }
    }
    shadowOffset=<0 0>
    shapedTextOptions {bOverride_TextFlowDirection=true, bOverride_TextShapingMethod=false, textFlowDirection=Culture, textShapingMethod=Auto}
   }
   slot {class=VerticalBoxSlot, properties {padding=<ltrb 0 0 0 24>}}
  }
  SizeBox_0 {
   widgetClassPath=&SizeBox
   parent=VerticalBox_1
   widget=SizeBox_0
   properties {bOverride_WidthOverride=true, widthOverride=784}
   slot {class=VerticalBoxSlot}
  }
  VerticalBox_0 {widgetClassPath=&VerticalBox, parent=SizeBox_0, widget=VerticalBox_0, slot {class=SizeBoxSlot}}
  Setting_TTS {
   widgetClassPath=/FortniteCoreUI/UI/Social/Chat/TextChat/WBP_TextChat_Entry_MessageBase
   bIsVariable=true
   parent=VerticalBox_0
   widget=Setting_TTS
   properties {
    holdData=&CommonUI_HoldData_Base_C
    minHeight=96
    minWidth=0
    navigation=&self:WidgetTree.Setting_TTS.WidgetNavigation_0
   }
   slot {class=VerticalBoxSlot}
  }
  ContentHB {widgetClassPath=&HorizontalBox, namedSlotHost=Setting_TTS, widget=ContentHB}
  Border_Title {
   widgetClassPath=&Border
   parent=ContentHB
   widget=Border_Title
   properties {
    background {
     bIsDynamicallyLoaded=false
     drawAs=NoDrawType
     imageType=NoImage
     mirroring=NoMirror
     resourceName=None
     resourceObject=null
     tiling=NoTile
     imageSize=<32 32>
     margin=<ltrb 0 0 0 0>
     outlineSettings {
      bUseBrushTransparency=false
      roundingType=HalfHeightRadius
      width=0
      color {colorUseRule=UseColor_Specified, specifiedColor=<rgba 0 0 0 0>}
      cornerRadii=<0 0 0 0>
     }
     tintColor {colorUseRule=UseColor_Specified, specifiedColor=<rgba 1 1 1 1>}
     uVRegion {bIsValid=false, max=<0 0>, min=<0 0>}
    }
    padding=<ltrb 0 0 0 0>
   }
   slot {
    class=HorizontalBoxSlot
    properties {horizontalAlignment=HAlign_Left, verticalAlignment=VAlign_Center, padding=<ltrb 30 0 0 0>, size {sizeRule=Fill, value=1}}
   }
  }
  Text_Title {
   widgetClassPath=&CommonTextBlock
   bIsVariable=true
   parent=Border_Title
   widget=Text_Title
   properties {
    justification=Center
    mobileFontSizeMultiplier=1.1
    scrollStyle=&TextScrollStyle-Base_C
    text=Read messages aloud
    colorAndOpacity {colorUseRule=UseColor_Specified, specifiedColor=<rgba 0.7529 0.7529 0.7529 1>}
    font {
     bForceMonospaced=false
     bMaterialIsStencil=false
     fontMaterial=null
     fontObject=&HeadingNow
     letterSpacing=0
     monospacedWidth=1
     size=16.5
     skewAmount=0
     typefaceFontName=75 Medium
     outlineSettings {
      bApplyOutlineToDropShadows=false
      bMiteredCorners=false
      bSeparateFillAlpha=false
      outlineMaterial=null
      outlineSize=0
      outlineColor=<rgba 0 0 0 1>
     }
    }
    shadowOffset=<0 0>
   }
   slot {class=BorderSlot, properties {padding=<ltrb 0 0 0 0>}}
  }
  SizeBox_1 {
   widgetClassPath=&SizeBox
   parent=ContentHB
   widget=SizeBox_1
   properties {bOverride_MinDesiredWidth=true, minDesiredWidth=282, visibility=HitTestInvisible}
   slot {
    class=HorizontalBoxSlot
    properties {horizontalAlignment=HAlign_Right, verticalAlignment=VAlign_Center, padding=<ltrb 0 12 30 12>, size {sizeRule=Fill, value=1}}
   }
  }
  TTS_Enabled {
   widgetClassPath=&Border
   bIsVariable=true
   parent=SizeBox_1
   widget=TTS_Enabled
   properties {
    horizontalAlignment=HAlign_Center
    verticalAlignment=VAlign_Bottom
    visibility=SelfHitTestInvisible
    background {
     bIsDynamicallyLoaded=false
     drawAs=Image
     imageType=NoImage
     mirroring=NoMirror
     resourceName=None
     resourceObject=/FortniteCoreUI/Settings/Materials/MI_SettingsRotator
     tiling=NoTile
     imageSize=<32 32>
     margin=<ltrb 0.5 0 0.5 0>
     outlineSettings {
      bUseBrushTransparency=false
      roundingType=HalfHeightRadius
      width=0
      color {colorUseRule=UseColor_Specified, specifiedColor=<rgba 0 0 0 0>}
      cornerRadii=<0 0 0 0>
     }
     tintColor {colorUseRule=UseColor_Specified, specifiedColor=<rgba 1 1 1 1>}
     uVRegion {bIsValid=false, max=<0 0>, min=<0 0>}
    }
    padding=<ltrb 0 0 0 10>
   }
   slot {class=SizeBoxSlot}
  }
  TTS_OnOff {
   widgetClassPath=&CommonTextBlock
   bIsVariable=true
   parent=TTS_Enabled
   widget=TTS_OnOff
   properties {
    justification=Center
    mobileFontSizeMultiplier=1.2
    scrollStyle=&TextScrollStyle-Base_C
    text=Off
    colorAndOpacity {colorUseRule=UseColor_Specified, specifiedColor=<rgba 0.75 0.75 0.75 1>}
    font {
     bForceMonospaced=false
     bMaterialIsStencil=false
     fontMaterial=null
     fontObject=&HeadingNow
     letterSpacing=0
     monospacedWidth=1
     size=18
     skewAmount=0
     typefaceFontName=75 Medium
     outlineSettings {
      bApplyOutlineToDropShadows=false
      bMiteredCorners=false
      bSeparateFillAlpha=false
      outlineMaterial=null
      outlineSize=0
      outlineColor=<rgba 0 0 0 1>
     }
    }
    shadowOffset=<0 0>
   }
   slot {class=BorderSlot, properties {horizontalAlignment=HAlign_Center, verticalAlignment=VAlign_Bottom, padding=<ltrb 0 0 0 10>}}
  }
  Setting_STT {
   widgetClassPath=&WBP_TextChat_Entry_MessageBase
   bIsVariable=true
   parent=VerticalBox_0
   widget=Setting_STT
   properties {
    holdData=&CommonUI_HoldData_Base_C
    minHeight=96
    minWidth=0
    navigation=&self:WidgetTree.Setting_STT.WidgetNavigation_0
   }
   slot {class=VerticalBoxSlot}
  }
  ContentHB_2 {widgetClassPath=&HorizontalBox, namedSlotHost=Setting_STT, widget=ContentHB_2}
  Border_Title_1 {
   widgetClassPath=&Border
   parent=ContentHB_2
   widget=Border_Title_1
   properties {
    background {
     bIsDynamicallyLoaded=false
     drawAs=NoDrawType
     imageType=NoImage
     mirroring=NoMirror
     resourceName=None
     resourceObject=null
     tiling=NoTile
     imageSize=<32 32>
     margin=<ltrb 0 0 0 0>
     outlineSettings {
      bUseBrushTransparency=false
      roundingType=HalfHeightRadius
      width=0
      color {colorUseRule=UseColor_Specified, specifiedColor=<rgba 0 0 0 0>}
      cornerRadii=<0 0 0 0>
     }
     tintColor {colorUseRule=UseColor_Specified, specifiedColor=<rgba 1 1 1 1>}
     uVRegion {bIsValid=false, max=<0 0>, min=<0 0>}
    }
    padding=<ltrb 0 0 0 0>
   }
   slot {
    class=HorizontalBoxSlot
    properties {horizontalAlignment=HAlign_Left, verticalAlignment=VAlign_Center, padding=<ltrb 30 0 0 0>, size {sizeRule=Fill, value=1}}
   }
  }
  Text_Title_1 {
   widgetClassPath=&CommonTextBlock
   bIsVariable=true
   parent=Border_Title_1
   widget=Text_Title_1
   properties {
    justification=Center
    mobileFontSizeMultiplier=1.1
    scrollStyle=&TextScrollStyle-Base_C
    text=Allow voice dictation
    colorAndOpacity {colorUseRule=UseColor_Specified, specifiedColor=<rgba 0.7529 0.7529 0.7529 1>}
    font {
     bForceMonospaced=false
     bMaterialIsStencil=false
     fontMaterial=null
     fontObject=&HeadingNow
     letterSpacing=0
     monospacedWidth=1
     size=16.5
     skewAmount=0
     typefaceFontName=75 Medium
     outlineSettings {
      bApplyOutlineToDropShadows=false
      bMiteredCorners=false
      bSeparateFillAlpha=false
      outlineMaterial=null
      outlineSize=0
      outlineColor=<rgba 0 0 0 1>
     }
    }
    shadowOffset=<0 0>
   }
   slot {class=BorderSlot, properties {padding=<ltrb 0 0 0 0>}}
  }
  SizeBox {
   widgetClassPath=&SizeBox
   parent=ContentHB_2
   widget=SizeBox
   properties {bOverride_MinDesiredWidth=true, minDesiredWidth=282, visibility=HitTestInvisible}
   slot {
    class=HorizontalBoxSlot
    properties {horizontalAlignment=HAlign_Right, verticalAlignment=VAlign_Center, padding=<ltrb 0 12 30 12>, size {sizeRule=Fill, value=1}}
   }
  }
  STT_Enabled {
   widgetClassPath=&Border
   bIsVariable=true
   parent=SizeBox
   widget=STT_Enabled
   properties {
    horizontalAlignment=HAlign_Center
    verticalAlignment=VAlign_Bottom
    visibility=SelfHitTestInvisible
    background {
     bIsDynamicallyLoaded=false
     drawAs=Image
     imageType=NoImage
     mirroring=NoMirror
     resourceName=None
     resourceObject=&MI_SettingsRotator
     tiling=NoTile
     imageSize=<32 32>
     margin=<ltrb 0.5 0 0.5 0>
     outlineSettings {
      bUseBrushTransparency=false
      roundingType=HalfHeightRadius
      width=0
      color {colorUseRule=UseColor_Specified, specifiedColor=<rgba 0 0 0 0>}
      cornerRadii=<0 0 0 0>
     }
     tintColor {colorUseRule=UseColor_Specified, specifiedColor=<rgba 1 1 1 1>}
     uVRegion {bIsValid=false, max=<0 0>, min=<0 0>}
    }
    padding=<ltrb 0 0 0 10>
   }
   slot {class=SizeBoxSlot}
  }
  STT_OnOff {
   widgetClassPath=&CommonTextBlock
   bIsVariable=true
   parent=STT_Enabled
   widget=STT_OnOff
   properties {
    justification=Center
    mobileFontSizeMultiplier=1.2
    scrollStyle=&TextScrollStyle-Base_C
    text=Off
    colorAndOpacity {colorUseRule=UseColor_Specified, specifiedColor=<rgba 0.75 0.75 0.75 1>}
    font {
     bForceMonospaced=false
     bMaterialIsStencil=false
     fontMaterial=null
     fontObject=&HeadingNow
     letterSpacing=0
     monospacedWidth=1
     size=18
     skewAmount=0
     typefaceFontName=75 Medium
     outlineSettings {
      bApplyOutlineToDropShadows=false
      bMiteredCorners=false
      bSeparateFillAlpha=false
      outlineMaterial=null
      outlineSize=0
      outlineColor=<rgba 0 0 0 1>
     }
    }
    shadowOffset=<0 0>
   }
   slot {class=BorderSlot, properties {horizontalAlignment=HAlign_Center, verticalAlignment=VAlign_Bottom, padding=<ltrb 0 0 0 10>}}
  }
  WBP_UIKit_Button_Regular {
   widgetClassPath=&WBP_UIKit_Button_Regular
   bIsVariable=true
   parent=VerticalBox_1
   widget=WBP_UIKit_Button_Regular
   properties {
    bHideInputAction=true
    buttonType=Quiet
    navigation=&self:WidgetTree.WBP_UIKit_Button_Regular.WidgetNavigation_0
    text=Done
   }
   slot {
    class=VerticalBoxSlot
    properties {horizontalAlignment=HAlign_Left, verticalAlignment=VAlign_Bottom, padding=<ltrb 0 48 0 48>, size {sizeRule=Fill, value=1}}
   }
  }
  VerticalBox_2 {
   widgetClassPath=&VerticalBox
   parent=HorizontalBox_1
   widget=VerticalBox_2
   slot {class=HorizontalBoxSlot, properties {padding=<ltrb 48 0 0 0>}}
  }
  Setting_Title {
   widgetClassPath=&CommonTextBlock
   bIsVariable=true
   parent=VerticalBox_2
   widget=Setting_Title
   properties {
    text=Speak text chat messages
    wrapTextAt=420
    font {
     bForceMonospaced=false
     bMaterialIsStencil=false
     fontMaterial=null
     fontObject=&HeadingNow
     letterSpacing=10
     monospacedWidth=1
     size=18
     skewAmount=0
     typefaceFontName=75 Medium
     outlineSettings {
      bApplyOutlineToDropShadows=false
      bMiteredCorners=false
      bSeparateFillAlpha=false
      outlineMaterial=&MI_UI_Block_Text_Outline
      outlineSize=2
      outlineColor=<rgba 0 0 0 1>
     }
    }
    shadowOffset=<0 0>
    shapedTextOptions {bOverride_TextFlowDirection=true, bOverride_TextShapingMethod=false, textFlowDirection=Culture, textShapingMethod=Auto}
   }
   slot {class=VerticalBoxSlot, properties {padding=<ltrb 0 0 0 24>}}
  }
  SizeBox_3 {
   widgetClassPath=&SizeBox
   parent=VerticalBox_2
   widget=SizeBox_3
   properties {bOverride_HeightOverride=true, bOverride_WidthOverride=true, heightOverride=440, widthOverride=416}
   slot {class=VerticalBoxSlot, properties {horizontalAlignment=HAlign_Left, verticalAlignment=VAlign_Top}}
  }
  RichTextBlock {
   widgetClassPath=&CommonRichTextBlock
   bIsVariable=true
   parent=SizeBox_3
   widget=RichTextBlock
   properties {
    autoWrapText=true
    bAutoCollapseWithEmptyText=true
    lineHeightPercentage=1.3
    mobileTextBlockScale=1.2
    scrollOrientation=Orient_Vertical
    scrollStyle=/FortUILibrary/Styles/Typography/Foundation/TextScrollStyle-NoFade.TextScrollStyle-NoFade_C
    text=Settings
    textStyleSet=/FortniteCoreUI/Settings/SettingsDetails/SettingsDescriptionStyles
    decoratorClasses [/FortniteCoreUI/Foundation/RichTextData/CommonRichTextDecorators.CommonRichTextDecorators_C]
    defaultTextStyleOverride {
     overflowPolicy=Clip
     transformPolicy=None
     colorAndOpacity {colorUseRule=UseColor_Specified, specifiedColor=<rgba 1 1 1 1>}
     font {
      bForceMonospaced=false
      bMaterialIsStencil=false
      fontMaterial=null
      fontObject=&NotoSans
      letterSpacing=0
      monospacedWidth=1
      size=24
      skewAmount=0
      typefaceFontName=NotoSans-Regular
      outlineSettings {
       bApplyOutlineToDropShadows=false
       bMiteredCorners=false
       bSeparateFillAlpha=false
       outlineMaterial=null
       outlineSize=0
       outlineColor=<rgba 0 0 0 1>
      }
     }
     highlightColor {colorUseRule=UseColor_Specified, specifiedColor=<rgba 0 0 0 1>}
     highlightShape {
      bIsDynamicallyLoaded=false
      drawAs=Image
      imageType=NoImage
      mirroring=NoMirror
      resourceName=None
      resourceObject=null
      tiling=NoTile
      imageSize=<32 32>
      margin=<ltrb 0 0 0 0>
      outlineSettings {
       bUseBrushTransparency=false
       roundingType=HalfHeightRadius
       width=0
       color {colorUseRule=UseColor_Specified, specifiedColor=<rgba 0 0 0 0>}
       cornerRadii=<0 0 0 0>
      }
      tintColor {colorUseRule=UseColor_Specified, specifiedColor=<rgba 1 1 1 1>}
      uVRegion {bIsValid=false, max=<0 0>, min=<0 0>}
     }
     selectedBackgroundColor {colorUseRule=UseColor_Foreground, specifiedColor=<rgba 1 0 1 1>}
     shadowColorAndOpacity=<rgba 0 0 0 1>
     shadowOffset=<0 0>
     strikeBrush {
      bIsDynamicallyLoaded=false
      drawAs=Image
      imageType=NoImage
      mirroring=NoMirror
      resourceName=None
      resourceObject=null
      tiling=NoTile
      imageSize=<32 32>
      margin=<ltrb 0 0 0 0>
      outlineSettings {
       bUseBrushTransparency=false
       roundingType=HalfHeightRadius
       width=0
       color {colorUseRule=UseColor_Specified, specifiedColor=<rgba 0 0 0 0>}
       cornerRadii=<0 0 0 0>
      }
      tintColor {colorUseRule=UseColor_Specified, specifiedColor=<rgba 1 1 1 1>}
      uVRegion {bIsValid=false, max=<0 0>, min=<0 0>}
     }
     underlineBrush {
      bIsDynamicallyLoaded=false
      drawAs=Image
      imageType=NoImage
      mirroring=NoMirror
      resourceName=None
      resourceObject=null
      tiling=NoTile
      imageSize=<32 32>
      margin=<ltrb 0 0 0 0>
      outlineSettings {
       bUseBrushTransparency=false
       roundingType=HalfHeightRadius
       width=0
       color {colorUseRule=UseColor_Specified, specifiedColor=<rgba 0 0 0 0>}
       cornerRadii=<0 0 0 0>
      }
      tintColor {colorUseRule=UseColor_Specified, specifiedColor=<rgba 1 1 1 1>}
      uVRegion {bIsValid=false, max=<0 0>, min=<0 0>}
     }
    }
   }
   slot {class=SizeBoxSlot}
  }
 ]
}
