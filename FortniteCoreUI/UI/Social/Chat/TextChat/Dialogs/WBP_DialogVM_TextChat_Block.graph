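# Text chat confirmation dialog for blocking a player, a CommonActivatableWidget built on the UIKit internal dialog shell.
# SetDescriptionText formats a username into the warning text, and PreConstruct fills it with a long sample name at design time.
# Confirm broadcasts OnClicked_ButtonConfirm with the button and dialog reference, leaving blocking and closing outside this widget.
# Cancel, bound to GenericBack, deactivates the dialog and is the default focus target.
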WidgetBlueprint /FortniteCoreUI/UI/Social/Chat/TextChat/Dialogs/WBP_DialogVM_TextChat_Block
graphs [
 BP_GetDesiredFocusTarget {nodes [
   Get Desired Focus Target()
   > Return Node!(execute, ReturnValue:Widget Object Reference=Get Button_Cancel() -> (Button_Cancel:WBP UIKit Button Regular Object Reference))
  ]}
 EventGraph {nodes [
   "On Clicked (Button_Cancel)"!() -> (OutputDelegate, then, Button:Common Button Base Object Reference) {is=title K2Node_ComponentBoundEvent}
   > DeactivateWidget()
   n1 = "On Clicked (Button_Confirm)"!() -> (OutputDelegate, then, Button:Common Button Base Object Reference) {is=title K2Node_ComponentBoundEvent}
   > Call On Clicked Button Confirm!(
    execute:Exec
    self:WBP Dialog VM Text Chat Block Object Reference
    Button:Common Button Base Object Reference=@n1.Button
    DialogRef:Object Reference=Self-Reference()
   ) -> (then:Exec) {is=title K2Node_CallDelegate}
   n2 = Event PreConstruct()
   > Branch(Condition=@n2.IsDesignTime)
   > SetDescriptionText!(execute, self, Username:Text=MakeLiteralText(Value="LOCTEXT(\"StrongUsernameStrongUsernameStrongUsername\")")) -> (then)
  ]}
 OnClicked_ButtonConfirm {nodes [OnClicked_ButtonConfirm!() -> (then, Button:Common Button Base Object Reference, DialogRef:Object Reference) {is=title K2Node_FunctionEntry}]}
 SetDescriptionText {nodes [
   n1 = SetDescriptionText!() -> (then, Username:Text) {is=title K2Node_FunctionEntry}
   > "SetText (Text)"(
    self=Get DescriptionText() -> (DescriptionText:Common Text Object Reference)
    InText=Format Text!(
     Format="LOCTEXT(\"Would you like to block {User}?\\r\\n\\r\\nBlocked players will be removed from your friends list, cannot invite you to their party, or voice chat or text chat with you. Players will not be notified that you blocked them.\")"
     User:Text=@n1
    ) -> (Result)
   )
  ]}
]
variables [OnClicked_ButtonConfirm {}]
widgets {
 info {
  inheritedWidgetCount=0
  namedSlotCount=1
  parentClass=&CommonActivatableWidget
  rootWidgetClass=/FortUILibrary/Systems/Dialog/WBP_UIKit_Dialog_Internal.WBP_UIKit_Dialog_Internal_C
  widgetCount=8
 }
 widgets [
  Dialog_Internal {
   widgetClassPath=/FortUILibrary/Systems/Dialog/WBP_UIKit_Dialog_Internal
   bIsVariable=true
   widget=Dialog_Internal
   properties {
    titleImageSizeX=64
    titleImageSizeY=64
    titleTextVerticalAligment=VAlign_Center
    useTitle=false
    buttons {
     each Button Horizontal Fill=false
     each Button Min Width=258
     each Button Size=Medium
     focuse Button=0
     space Between Buttons=24
     cTA Buttons Types []
     group Placement {
      horizontal Alignment=HAlign_Fill
      vertical Alignment=VAlign_Bottom
      grid {column=1, column Span=0, layer=0, row=2, row Span=0, nudge=<0 0>}
      padding=<ltrb 80 64 80 48>
     }
    }
    dialogSIze {
     height Override {enable=false, value=592}
     max Aspect Ratio {enable=false, value=0}
     max Desired Height {enable=true, value=800}
     max Desired Width {enable=false, value=0}
     min Aspect Ratio {enable=false, value=0}
     min Desired Height {enable=true, value=592}
     min Desired Width {enable=false, value=0}
     width Override {enable=true, value=855}
    }
    titleImagePadding=<ltrb 10 0 15 0>
    titleProperties {
     auto Wrap Text=true
     justification=Left
     line Height Percentage=1
     mobile Font Size Multiplier=1.3
     transform Policy=ToUpper
     wrap Text At=700
     color {colorUseRule=UseColor_Specified, specifiedColor=<rgba 1 1 1 1>}
     font {
      bForceMonospaced=false
      bMaterialIsStencil=false
      fontMaterial=null
      fontObject=&HeadingNow
      letterSpacing=0
      monospacedWidth=1
      size=25.5
      skewAmount=0
      typefaceFontName=86 Bold Italic
      outlineSettings {
       bApplyOutlineToDropShadows=false
       bMiteredCorners=false
       bSeparateFillAlpha=false
       outlineMaterial=null
       outlineSize=0
       outlineColor=<rgba 0 0 0 1>
      }
     }
     placement {
      horizontal Alignment=HAlign_Fill
      vertical Alignment=VAlign_Top
      grid {column=1, column Span=0, layer=0, row=0, row Span=0, nudge=<0 0>}
      padding=<ltrb 80 80 80 32>
     }
    }
   }
  }
  Overlay_0 {widgetClassPath=&Overlay, namedSlotHost=Dialog_Internal, widget=Overlay_0}
  GridPanel_0 {
   widgetClassPath=&GridPanel
   parent=Overlay_0
   widget=GridPanel_0
   properties {columnFill [1], rowFill [0,1,0]}
   slot {class=OverlaySlot, properties {horizontalAlignment=HAlign_Fill, verticalAlignment=VAlign_Fill, padding=<ltrb 80 80 80 48>}}
  }
  TitleText {
   widgetClassPath=&CommonTextBlock
   bIsVariable=true
   parent=GridPanel_0
   widget=TitleText
   properties {
    autoWrapText=true
    bIsScrollingEnabled=false
    style=&TextStyle-Base_C
    text=Block Player
    wrappingPolicy=AllowPerCharacterWrapping
    font {
     bForceMonospaced=false
     bMaterialIsStencil=false
     fontMaterial=null
     fontObject=&NotoSans
     letterSpacing=0
     monospacedWidth=1
     size=24
     skewAmount=0
     typefaceFontName=NotoSans-Regular
     outlineSettings {
      bApplyOutlineToDropShadows=false
      bMiteredCorners=false
      bSeparateFillAlpha=false
      outlineMaterial=null
      outlineSize=0
      outlineColor=<rgba 0 0 0 1>
     }
    }
    shadowOffset=<0 0>
   }
   slot {class=GridSlot, properties {horizontalAlignment=HAlign_Left, padding=<ltrb 0 0 0 32>}}
  }
  DescriptionText {
   widgetClassPath=&CommonTextBlock
   bIsVariable=true
   parent=GridPanel_0
   widget=DescriptionText
   properties {
    autoWrapText=true
    style=&TextStyle-Base_C
    text="Would you like to block {User}?\r\n\r\nBlocked players will be removed from your friends list, cannot invite you to their party, or voice chat or text chat with you. Players will not be notified that you blocked them."
    font {
     bForceMonospaced=false
     bMaterialIsStencil=false
     fontMaterial=null
     fontObject=&NotoSans
     letterSpacing=0
     monospacedWidth=1
     size=24
     skewAmount=0
     typefaceFontName=NotoSans-Regular
     outlineSettings {
      bApplyOutlineToDropShadows=false
      bMiteredCorners=false
      bSeparateFillAlpha=false
      outlineMaterial=null
      outlineSize=0
      outlineColor=<rgba 0 0 0 1>
     }
    }
    shadowOffset=<0 0>
   }
   slot {class=GridSlot, properties {row=1, padding=<ltrb 0 0 0 60>}}
  }
  GridPanel_1 {
   widgetClassPath=&GridPanel
   parent=GridPanel_0
   widget=GridPanel_1
   properties {columnFill [1,1]}
   slot {class=GridSlot, properties {row=2}}
  }
  Button_Confirm {
   widgetClassPath=/FortUILibrary/UIKit/Buttons/WBP_UIKit_Button_Loud
   bIsVariable=true
   parent=GridPanel_1
   widget=Button_Confirm
   properties {
    bDisplayInputActionWhenNotInteractable=false
    bShouldUseFallbackDefaultInputAction=false
    navigation=&self:WidgetTree.Button_Confirm.WidgetNavigation_0
    text=Block
   }
   slot {class=GridSlot}
  }
  Button_Cancel {
   widgetClassPath=&WBP_UIKit_Button_Regular
   bIsVariable=true
   parent=GridPanel_1
   widget=Button_Cancel
   properties {
    bDisplayInputActionWhenNotInteractable=false
    bShouldUseFallbackDefaultInputAction=false
    buttonType=Quiet
    navigation=&self:WidgetTree.Button_Cancel.WidgetNavigation_0
    text=Cancel
    triggeringInputAction {dataTable=&InputActionDataTable, rowName=GenericBack}
   }
   slot {class=GridSlot, properties {column=1, padding=<ltrb 24 0 0 0>}}
  }
 ]
}
