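# Text chat toggle button, derived from WBP_UIKit_ButtonWithToggle, that marks a player for blocking.
# PreConstruct adjusts layout and registers an extra image block with a background material, shown only on hover.
# The list item is cached as a block-player view model, and clicking flips its marked-for-block flag.
# The new flag drives block selection and a visual state transition that depends on user focus.
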WidgetBlueprint /FortniteCoreUI/UI/Social/Chat/TextChat/WBP_TextChat_ToggleButton
graphs [
 EventGraph {nodes [
   n1 = Event PreConstruct()
   > Parent: PreConstruct(IsDesignTime=@n1.IsDesignTime)
   > SetGridSlotColumnAndRow!(
    execute
    Widget:Widget Object Reference=Get Block_Text() -> (Block_Text:WBP UIKit Block Text Object Reference)
    Row:int=1
    Row Span:int=1
    Column:int=1
    Column Span:int=1
   ) -> (then)
   > SetGridSlotPadding!(
    execute
    Widget:Widget Object Reference=Get Slot_Main() -> (Slot_Main:Named Slot Object Reference)
    Padding:Margin Structure=Get ButtonOuterBlockPadding() -> (ButtonOuterBlockPadding:Margin Structure)
   ) -> (then)
   n2 = Get Block_Image() -> (Block_Image:WBP UIKit Block Image Object Reference)
   > RegisterModularBlock!(execute, self:UIKit Modular Button Object Reference, Block:UIKit Block Object Reference=@n2) -> (then)
   > SetMaterial!(
    execute
    self:WBP UIKit Block Image Object Reference=@n2
    InMaterial:"Material Instance Object Reference (by ref)"=Get BlockImageMaterial() -> (BlockImageMaterial:Material Instance Object Reference)
   ) -> (then)
   > SetBlockVisibility!(execute, self:WBP UIKit Block Base Object Reference=Get Block_Image(), Visible:bool=false) -> (then)
   Event On Hovered()
   > Parent: On Hovered() {is=title K2Node_CallParentFunction}
   n3 = Get Block_Image()
   > SetBlockVisibility(self=@n3, Visible=true)
   Event On Unhovered()
   > Parent: On Unhovered() {is=title K2Node_CallParentFunction}
   > SetBlockVisibility(self=@n3, Visible=false)
   n4 = Event OnListItemObjectSet()
   > n5 = Cast To FortBlockPlayerVM!(execute, Object=@n4.ListItemObject) -> (then, CastFailed, AsBlock Players View Model:Block Players View Model Object Reference) {is=title K2Node_DynamicCast}
   > Set FortBlockPlayerVM(FortBlockPlayerVM:Block Players View Model Object Reference=@n5) -> (Output_Get:Block Players View Model Object Reference)
   Event On Clicked()
   n6 = Get FortBlockPlayerVM() -> (FortBlockPlayerVM:Block Players View Model Object Reference)
   > n7 = Set bIsMarkedForBlock!(
    execute
    bIsMarkedForBlock:bool=NOT Boolean(A=Get bIsMarkedForBlock!(self:Block Players View Model Object Reference=@n6) -> (bIsMarkedForBlock:bool))
    self:Block Players View Model Object Reference=@n6
   ) -> (then, Output_Get:bool)
   > UpdsteBlocksDisplayState!(execute, self, bMarkedForBlock:bool=@n7) -> (then)
   Event On Selected()
   Event On Deselected()
  ]}
 SetPlayernameText {nodes [
   n1 = SetPlayernameText!() -> (then, PlayerName:String) {is=title K2Node_FunctionEntry}
   > SetText!(execute, self:WBP UIKit Button Generic Base Object Reference, Text:Text="To Text (String)"(InString=@n1)) -> (then)
  ]}
 UpdsteBlocksDisplayState {nodes [
   n1 = UpdsteBlocksDisplayState!() -> (then, bMarkedForBlock:bool=~) {is=title K2Node_FunctionEntry}
   > SetBlocksIsSelected!(execute, self:WBP UIKit Button Generic Base Object Reference, IsSelected:bool=@n1) -> (then)
   > n2 = Branch(Condition=HasUserFocus!(self:Widget Object Reference, PlayerController:Player Controller Object Reference=GetOwningPlayer()) -> (ReturnValue:bool))
   > RequestTransitionOnBlocks!(
    execute
    self:UIKit Modular Button Object Reference
    State:EUIKitBlockVisualState Enum=Select(Option 0:EUIKitBlockVisualState Enum=DeselectedFocused, Option 1:EUIKitBlockVisualState Enum=Focused, Index=@n1) -> (ReturnValue:EUIKitBlockVisualState Enum)
   ) -> (then)
   RequestTransitionOnBlocks(execute=@n2.else, State=Select(Option 0=DeselectedIdle, Option 1=Selected, Index=@n1))
  ]}
]
variables [
 BlockImageMaterial {type=MaterialInstance, default=/FortniteCoreUI/UI/Social/Chat/TextChat/MaterialInstances/MI_UI_TextChat_Toggle_Background}
 ButtonOuterBlockPadding {type=Margin, default=<ltrb 16 16 16 16>}
]
widgets {
 info {
  inheritedWidgetCount=0
  namedSlotCount=1
  parentClass=/FortUILibrary/UIKit/Buttons/WBP_UIKit_ButtonWithToggle.WBP_UIKit_ButtonWithToggle_C
  rootWidgetClass=null
  widgetCount=2
 }
 widgets [
  GridPanel_45 {widgetClassPath=&GridPanel, widget=GridPanel_45, properties {columnFill [1], rowFill [1]}}
  Block_Image {
   widgetClassPath=/FortUILibrary/Systems/ModularButton/Blocks/WBP_UIKit_Block_Image
   bIsVariable=true
   parent=GridPanel_45
   widget=Block_Image
   slot {class=GridSlot, properties {padding=<ltrb -64 0 0 0>}}
  }
 ]
}
