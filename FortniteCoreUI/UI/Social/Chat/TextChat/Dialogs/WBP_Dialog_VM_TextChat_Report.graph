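# Activatable dialog for reporting Fortnite text chat, groups, messages or invites, then optionally blocking players.
# Callers set the reported channel or message; on Construct it tailors text by content type and starts FortTextChatReportVM's report.
# SetReportableState drives nested visibility switchers across submit, loading, error, submitted and block panels, with desired focus following the active page.
# Buttons send, cancel or block through the view model; closing broadcasts OnReportClosed.
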WidgetBlueprint /FortniteCoreUI/UI/Social/Chat/TextChat/Dialogs/WBP_Dialog_VM_TextChat_Report
graphs [
 BP_GetDesiredFocusTarget {nodes [
   Get Desired Focus Target()
   n1 = Get S_Content() -> (S_Content:Common Visibility Switcher Object Reference)
   > Return Node!(
    execute
    ReturnValue:Widget Object Reference=Select!(
     Option 0:Widget Object Reference=Get Button_SubmitReport() -> (Button_SubmitReport:WBP UIKit Button Loud Object Reference)
     Option 1:Widget Object Reference=Get Button_PlayerBlock() -> (Button_PlayerBlock:WBP UIKit Button Regular Object Reference)
     Option 2:Widget Object Reference=Get CommonTileView() -> (CommonTileView:Common Tile View Object Reference)
     Index:int=GetActiveWidgetIndex!(self:Common Visibility Switcher Object Reference=@n1) -> (ReturnValue:int=0)
    ) -> (ReturnValue:Widget Object Reference)
   )
  ]}
 EventGraph {nodes [
   n1 = "On Clicked (Button_Nevermind)"!() -> (OutputDelegate, then, Button:Common Button Base Object Reference) {is=title K2Node_ComponentBoundEvent}
   n2 = Get FortTextChatReportVM() -> (FortTextChatReportVM:Text Chat Report View Model Object Reference)
   n3 = CancelReport(execute=[@n1.then, @n5.then], self:Text Chat Report View Model Object Reference=@n2)
   Call On Report Closed!(
    execute:Exec=[@n3.then, @n10.then, @n4.then]
    self:WBP Dialog VM Text Chat Report Object Reference
    ReportState:ETextReportState Enum=Get ReportState!(self:Text Chat Report View Model Object Reference=@n2) -> (ReportState:ETextReportState Enum=BeginReport)
   ) -> (then:Exec) {is=title K2Node_CallDelegate}
   > DeactivateWidget()
   n4 = "On Clicked (Button_Close)"!() -> (OutputDelegate, then, Button:Common Button Base Object Reference) {is=title K2Node_ComponentBoundEvent}
   n5 = "On Clicked (Button_CloseError)"!() -> (OutputDelegate, then, Button:Common Button Base Object Reference) {is=title K2Node_ComponentBoundEvent}
   n6 = "On Entry Initialized (CommonTileView)"!() -> (OutputDelegate, then, Item:Object Reference, Widget:User Widget Object Reference) {is=title K2Node_ComponentBoundEvent}
   > n7 = Cast To WBP_TextChat_ToggleButton!(execute, Object=@n6.Widget) -> (then, CastFailed, AsWBP Text Chat Toggle Button:WBP Text Chat Toggle Button Object Reference) {is=title K2Node_DynamicCast}
   > SetIsSelected!(execute, self:Common Button Base Object Reference=@n7, InSelected:bool=false, bGiveClickFeedback:bool=true) -> (then)
   "On Clicked (Button_SubmitReport)"!() -> (OutputDelegate, then, Button:Common Button Base Object Reference) {is=title K2Node_ComponentBoundEvent}
   > SendReport(self:Text Chat Report View Model Object Reference=Get FortTextChatReportVM())
   "On Item Clicked (CommonTileView)"!() -> (OutputDelegate, then, Item:Object Reference) {is=title K2Node_ComponentBoundEvent}
   > SetIsEnabled(self=Get Button_Block() -> (Button_Block:WBP UIKit Button Loud Object Reference), bInIsEnabled=false)
   > n8 = For Each Loop(Array:Array of Object References=GetListItems!(self:List View Object Reference=Get CommonTileView()) -> (ReturnValue:Array of Object References)) -> (Array Element:Object Reference)
   > n9 = Cast To FortBlockPlayerVM!(execute, Object=@n8) -> (then, CastFailed, AsBlock Players View Model:Block Players View Model Object Reference) {is=title K2Node_DynamicCast}
   > Branch(Condition=Get bIsMarkedForBlock!(self:Block Players View Model Object Reference=@n9) -> (bIsMarkedForBlock:bool))
   > SetIsEnabled(self=Get Button_Block(), bInIsEnabled=true)
   "On Clicked (Button_PlayerBlock)"!() -> (OutputDelegate, then, Button:Common Button Base Object Reference) {is=title K2Node_ComponentBoundEvent}
   > SetActiveWidget(self=Get S_Content(), Widget=Get GP_BlockPlayer() -> (GP_BlockPlayer:Grid Panel Object Reference))
   > RequestRefreshFocus(self:Common Activatable Widget Object Reference)
   n10 = "On Clicked (Button_Close_Block)"!() -> (OutputDelegate, then, Button:Common Button Base Object Reference) {is=title K2Node_ComponentBoundEvent}
   "On Clicked (Button_Block)"!() -> (OutputDelegate, then, Button:Common Button Base Object Reference) {is=title K2Node_ComponentBoundEvent}
   > BlockPlayers(self:Text Chat Report View Model Object Reference=Get FortTextChatReportVM())
   > DeactivateWidget()
   n11 = Event PreConstruct()
   > n12 = Sequence!(execute) -> (then_0, then_1, then_2:Exec)
   > SetIsEnabled(self=Get Button_Block(), bInIsEnabled=false)
   > SetErrorText!(execute, self, Header:Text="LOCTEXT(\"Network Error\")", Message:Text="LOCTEXT(\"please check your internet connection \")") -> (then)
   Branch(execute=@n12.then_1, Condition=@n11.IsDesignTime)
   > SetReportSubmittedID!(execute, self, ReportID:String=999-999-999) -> (then)
   SetCompactUIStyle(execute=@n12.then_2) {is=title K2Node_CallFunction}
   Event Construct()
   n13 = GetReportedContentType!(
    self:Text Chat Report View Model Object Reference=Get FortTextChatReportVM()
    ConversationId:Name=Get ReportedChanneld() -> (ReportedChanneld:Name=None)
    ChannelType:EFortTextChatChannelType Enum=Get ReportedChannelType() -> (ReportedChannelType:EFortTextChatChannelType Enum=All)
    MessageId:Name=Get ReportedMessageId() -> (ReportedMessageId:Name=None)
   ) -> (ReturnValue:ETextChatReportedContentType Enum=Chat)
   > SetReportChatTitle!(execute, self, ReportedContentType:ETextChatReportedContentType Enum=@n13) -> (then)
   > SetReportChatSubtitle!(execute, self, ReportedContentType:ETextChatReportedContentType Enum=@n13) -> (then)
   > SetReportChatDescriptionText!(execute, self, ReportedContentType:ETextChatReportedContentType Enum=@n13) -> (then)
   > StartReport!(
    execute
    self:Text Chat Report View Model Object Reference=Get FortTextChatReportVM()
    InReportedChannelId:Name=Get ReportedChanneld() -> (ReportedChanneld=None)
    InReportedMessageId:Name=Get ReportedMessageId() -> (ReportedMessageId=None)
    InReportedChannelType:EFortTextChatChannelType Enum=Get ReportedChannelType() -> (ReportedChannelType=All)
    bInReportMostRecentMessages:bool=Get ReportMostRecentMessages() -> (ReportMostRecentMessages:bool)
   ) -> (then)
   Event Destruct()
   > RemoveTagFromInputMode!(
    execute
    self:Enhanced Input Subsystem Interface Interface=Get EnhancedInputLocalPlayerSubsystem() -> (ReturnValue:Enhanced Input Local Player Subsystem Object Reference) {is=title K2Node_GetSubsystem}
    TagToRemove:"Gameplay Tag Structure (by ref)"=MakeLiteralGameplayTag!(Value:Gameplay Tag Structure=(TagName="UI.TextChat.Disabled")) -> (ReturnValue:Gameplay Tag Structure)
    Options:"Modify Context Options Structure (by ref)"=(bIgnoreAllPressedKeysUntilRelease=True,bForceImmediately=False,bNotifyUserSettings=False)
   ) -> (then)
  ]}
 FillOutBlockPlayers {nodes [
   n1 = FillOutBlockPlayers!() -> (then, PlayerToBlock:Array of Block Players View Model Object References) {is=title K2Node_FunctionEntry}
   > ClearListItems(self=Get CommonTileView())
   > Branch(Condition=Is Not Empty!(TargetArray:Array of Block Players View Model Object References=@n1) -> (ReturnValue:bool))
   > Set List Items!(execute, self:List View Object Reference=Get CommonTileView(), InListItems:Array of Object References=@n1) -> (then) {is=title K2Node_CallFunction}
  ]}
 OnMouseWheel {nodes [
   OnMouseWheel!() -> (then, MyGeometry:Geometry Structure, MouseEvent:"Pointer Event Structure (by ref)") {is=title K2Node_FunctionEntry}
   > n1 = Branch(Condition=HasAnyUserFocus!(self:Widget Object Reference=Self-Reference()) -> (ReturnValue:bool))
   > Return Node(ReturnValue:Event Reply Structure=Unhandled())
   Return Node(execute=@n1.else, ReturnValue=Handled())
  ]}
 OnNavigateDownFromBlockList {nodes [
   OnNavigateDownFromBlockList!() -> (then, Navigation:EUINavigation Enum) {is=title K2Node_FunctionEntry}
   n1 = Get Button_Block()
   > n2 = Branch(Condition=Get bIsEnabled!(self:Widget Object Reference=@n1) -> (bIsEnabled:bool))
   > Return Node(ReturnValue:Widget Object Reference=@n1)
   Return Node(execute=@n2.else, ReturnValue=Get Button_Close_Block() -> (Button_Close_Block:WBP UIKit Button Regular Object Reference))
  ]}
 OnReportClosed {nodes [OnReportClosed!() -> (then, ReportState:ETextReportState Enum=BeginReport) {is=title K2Node_FunctionEntry}]}
 SetCompactUIStyle {nodes [
   SetCompactUIStyle() {is=title K2Node_FunctionEntry}
   > Branch(Condition=IsMobileGame())
   > SetMaxDesiredWidth!(execute, self:Size Box Object Reference=Get SB_Mobile() -> (SB_Mobile:Size Box Object Reference), InMaxDesiredWidth:float=720) -> (then)
  ]}
 SetErrorText {nodes [
   n1 = SetErrorText!() -> (then, Header:Text, Message:Text) {is=title K2Node_FunctionEntry}
   > "SetText (Text)"(self=Get Text_ErrorHeader() -> (Text_ErrorHeader:Common Text Object Reference), InText=@n1)
   > "SetText (Text)"(self=Get Text_ErrorMessage() -> (Text_ErrorMessage:Common Text Object Reference), InText=@n1.Message)
  ]}
 SetReportChatDescriptionText {nodes [
   n1 = SetReportChatDescriptionText!() -> (then, ReportedContentType:ETextChatReportedContentType Enum=Chat) {is=title K2Node_FunctionEntry}
   > SetText(
    self=Get ReportChat_DescriptionText() -> (ReportChat_DescriptionText:Common Rich Text Block Object Reference)
    InText=Select!(
     Chat:"Text (by ref)"=MakeLiteralText(Value="LOCTEXT(\"Your report will submit recent reportable messages from your chat to Epic Games for review.\\r\\n\\r\\nOnce submitted, Epic will review the conduct of all participants in the text chat conversation, including yours, for any violations of our Community Rules. Players will not be notified that you are the reporter.\")")
     Group:"Text (by ref)"=MakeLiteralText(Value="LOCTEXT(\"Your report will submit recent reportable messages and group names from your group to Epic Games for review.\\r\\n\\r\\nOnce submitted, Epic will review the conduct of all participants in the text chat conversation, including yours, for any violations of our Community Rules. Players will not be notified that you are the reporter.\\r\\n\")")
     Message:"Text (by ref)"=MakeLiteralText(Value="LOCTEXT(\"Your report will submit the selected message, recent reportable messages, and group names from your group to Epic Games for review.\\r\\n\\r\\nOnce submitted, Epic will review the conduct of all participants in the text chat conversation, including yours, for any violations of our Community Rules. Players will not be notified that you are the reporter.\\r\\n\")")
     Invitation:"Text (by ref)"=MakeLiteralText(Value="LOCTEXT(\"Your report will submit the invite, the group you were invited to, and the player who invited you to Epic Games for review.\\r\\n\\r\\nOnce submitted, Epic will review the report for any violations of our Community Rules. Players will not be notified that you are the reporter.\")")
     Party:"Text (by ref)"=MakeLiteralText(Value="LOCTEXT(\"Your report will submit all reportable text chat messages in your Party Channel to Epic Games for review.\\r\\n\\r\\nOnce submitted, Epic will review the conduct of all participants in the text chat conversation, including yours, for any violations of our Community Rules. Players will not be notified that you are the reporter.\\r\\n\")")
     Game:"Text (by ref)"=MakeLiteralText(Value="LOCTEXT(\"Your report will submit all reportable text chat messages in your Game Channel to Epic Games for review.\\r\\n\\r\\nOnce submitted, Epic will review the conduct of all participants in the text chat conversation, including yours, for any violations of our Community Rules. Players will not be notified that you are the reporter.\\r\\n\")")
     Index:ETextChatReportedContentType Enum=@n1
    ) -> (ReturnValue:"Text (by ref)")
   )
   Reroute Node(InputPin:Wildcard) -> (OutputPin:Wildcard)
  ]}
 SetReportChatSubtitle {nodes [
   n1 = SetReportChatSubtitle!() -> (then, ReportedContentType:ETextChatReportedContentType Enum=Chat) {is=title K2Node_FunctionEntry}
   n2 = Format Text!(
    Format="LOCTEXT(\"{Current Channel} · {Member Count} members\")"
    Current Channel:Text=Get ReportedChannellName() -> (ReportedChannellName:Text)
    Member Count:int=GetNumGroupMembers!(
     self:Text Chat Report View Model Object Reference=Get FortTextChatReportVM()
     ConversationId:Name=Get ReportedChanneld() -> (ReportedChanneld=None)
    ) -> (ReturnValue:int=0)
   ) -> (Result)
   > "SetText (Text)"(
    self=[
     Get BlockPlayer_Subtitle() -> (BlockPlayer_Subtitle:Common Text Object Reference)
     Get ReportChat_Subtitle() -> (ReportChat_Subtitle:Common Text Object Reference)
    ]
    InText=Select(Chat:Text=Get ReportedChannellName(), Group:Text=@n2, Message:Text=@n2, Invitation:Text, Party:Text, Game:Text, Index=@n1) -> (ReturnValue:Text)
   )
  ]}
 SetReportChatTitle {nodes [
   n1 = SetReportChatTitle!() -> (then, ReportedContentType:ETextChatReportedContentType Enum=Chat) {is=title K2Node_FunctionEntry}
   > "SetText (Text)"(
    self=Get ReportChat_Title() -> (ReportChat_Title:Common Text Object Reference)
    InText=Select(
     Chat="LOCTEXT(\"REPORT CHAT\")"
     Group="LOCTEXT(\"REPORT CHAT\")"
     Message="LOCTEXT(\"REPORT CHAT\")"
     Invitation="LOCTEXT(\"REPORT INVITE\")"
     Party="LOCTEXT(\"REPORT CHAT: PARTY CHANNEL\")"
     Game="LOCTEXT(\"REPORT CHAT: GAME CHANNEL\")"
     Index=@n1
    )
   )
  ]}
 SetReportSubmittedID {nodes [
   n1 = SetReportSubmittedID!() -> (then, ReportID:String) {is=title K2Node_FunctionEntry}
   > SetText(
    self=Get Text_ReportSubmitted() -> (Text_ReportSubmitted:Common Rich Text Block Object Reference)
    InText=Format Text!(
     Format="LOCTEXT(\"Thank you for your report. To track the status of your report and any enforcement actions taken, check your My Reports page at <URL>https://epic.gm/my-reports/</> (ID: {XXX-XXX-XXX})\\r\\n\\r\\nReports help us to ensure that Fortnite remains a fair, safe and enjoyable game for all players.\")"
     XXX-XXX-XXX:String=@n1
    ) -> (Result)
   )
  ]}
 SetReportableState {nodes [
   n1 = SetReportableState!() -> (then, ReportState:ETextReportState Enum=BeginReport) {is=title K2Node_FunctionEntry}
   > n2 = Switch on ETextReportState!(execute, Selection:ETextReportState Enum=@n1) -> (BeginReport:Exec, ReportPending:Exec, ReportSuccessful:Exec, ReportFailed:Exec, BlockPlayers:Exec) {is=title K2Node_SwitchEnum}
   > SetActiveWidget(self=Get S_Content(), Widget=Get GP_SubmitReport() -> (GP_SubmitReport:Grid Panel Object Reference))
   > RequestRefreshFocus()
   SetActiveWidget(
    execute=@n2.ReportPending
    self=Get S_ReportContent() -> (S_ReportContent:Common Visibility Switcher Object Reference)
    Widget=Get GP_Loading() -> (GP_Loading:Grid Panel Object Reference)
   )
   SetActiveWidget(execute=@n2.ReportFailed, self=Get S_ReportContent(), Widget=Get GP_Error() -> (GP_Error:Grid Panel Object Reference))
   > SetFocus(self=Get Button_CloseError() -> (Button_CloseError:WBP UIKit Button Loud Object Reference))
   SetActiveWidget(execute=@n2.ReportSuccessful, self=Get S_Content(), Widget=Get GP_ReportSubmitted() -> (GP_ReportSubmitted:Grid Panel Object Reference))
   > RequestRefreshFocus()
   SetActiveWidget(execute=@n2.BlockPlayers, self=Get S_Content(), Widget=Get GP_BlockPlayer())
   > RequestRefreshFocus()
  ]}
 SetReportedChannel {nodes [
   n1 = SetReportedChannel!() -> (
    then
    ReportedChannelId:Name
    ReportedChannelType:EFortTextChatChannelType Enum=All
    ReportedChannelName:Text
    ReportedChannelMemberCount:int
    ReportMostRecentMessages:bool=~
   ) {is=title K2Node_FunctionEntry}
   > Set ReportedChanneld(ReportedChanneld:Name=@n1) -> (Output_Get:Name=None)
   > Set ReportedChannelType(ReportedChannelType:EFortTextChatChannelType Enum=@n1.ReportedChannelType) -> (Output_Get:EFortTextChatChannelType Enum=All)
   > Set ReportedChannellName(ReportedChannellName:Text=@n1.ReportedChannelName) -> (Output_Get:Text)
   > Set ReportedChannelMemberCount(ReportedChannelMemberCount:int=@n1.ReportedChannelMemberCount) -> (Output_Get:int=0)
   > Set ReportedMessageId(ReportedMessageId:Name=None) -> (Output_Get:Name=None)
   > Set ReportMostRecentMessages(ReportMostRecentMessages:bool=@n1.ReportMostRecentMessages) -> (Output_Get:bool)
  ]}
 SetReportedChannelVM {nodes [
   n1 = SetReportedChannelVM!() -> (then, ReportedChannelVM:Fort Text Chat Channel VM Object Reference, ReportMostRecentMessages:bool=~) {is=title K2Node_FunctionEntry}
   > SetReportedChannel!(
    execute
    self
    ReportedChannelId:Name=Get ChannelId!(self:Fort Text Chat Channel VM Object Reference=@n1) -> (ChannelId:Name=None)
    ReportedChannelType:EFortTextChatChannelType Enum=Get ChannelType!(self:Fort Text Chat Channel VM Object Reference=@n1) -> (ChannelType:EFortTextChatChannelType Enum=All)
    ReportedChannelName:Text="To Text (String)"(InString=Get ChannelName!(self:Fort Text Chat Channel VM Object Reference=@n1) -> (ChannelName:String))
    ReportedChannelMemberCount:int=0
    ReportMostRecentMessages:bool=@n1.ReportMostRecentMessages
   ) -> (then) {is=title K2Node_CallFunction}
  ]}
 SetReportedConversationVM {nodes [
   n1 = SetReportedConversationVM!() -> (then, ReportedConversationVM:Fort Text Chat Conversation VM Object Reference) {is=title K2Node_FunctionEntry}
   > SetReportedChannel(
    ReportedChannelId=Get ConversationId!(self:Fort Text Chat Conversation VM Object Reference=@n1) -> (ConversationId:Name=None)
    ReportedChannelType=Persistent
    ReportedChannelName=Get ConversationName!(self:Fort Text Chat Conversation VM Object Reference=@n1) -> (ConversationName:Text)
    ReportedChannelMemberCount=Length(TargetArray:Array of Fort Social User Viewmodel Object References=Get Members!(self:Fort Text Chat Conversation VM Object Reference=@n1) -> (Members:Array of Fort Social User Viewmodel Object References))
    ReportMostRecentMessages=false
   )
  ]}
 SetReportedMessageId {nodes [
   n1 = SetReportedMessageId!() -> (then, MessageId:Name) {is=title K2Node_FunctionEntry}
   > Set ReportedMessageId(ReportedMessageId=@n1) -> (Output_Get=None)
  ]}
]
variables [
 OnReportClosed {}
 ReportedChanneld {type=string, default=None}
 ReportedChannellName {type=string, default=""}
 ReportedChannelMemberCount {type=integer}
 ReportedChannelType {type=EFortTextChatChannelType, default=All}
 ReportedConversationType {type=EFortTextChatPersistentConversationType, default=EpicGroup}
 ReportedMessageId {type=string, default=None}
 ReportMostRecentMessages {type=boolean}
]
widgets {
 info {
  inheritedWidgetCount=0
  namedSlotCount=1
  parentClass=&CommonActivatableWidget
  rootWidgetClass=/FortUILibrary/Systems/Dialog/WBP_UIKit_Dialog_Internal.WBP_UIKit_Dialog_Internal_C
  widgetCount=41
 }
 widgets [
  Dialog_Internal {
   widgetClassPath=/FortUILibrary/Systems/Dialog/WBP_UIKit_Dialog_Internal
   bIsVariable=true
   widget=Dialog_Internal
   properties {
    bIsFocusable=true
    titleImageSizeX=64
    titleImageSizeY=64
    titleTextVerticalAligment=VAlign_Center
    useTitle=false
    buttons {
     each Button Horizontal Fill=false
     each Button Min Width=258
     each Button Size=Medium
     focuse Button=0
     space Between Buttons=24
     cTA Buttons Types []
     group Placement {
      horizontal Alignment=HAlign_Fill
      vertical Alignment=VAlign_Bottom
      grid {column=1, column Span=0, layer=0, row=2, row Span=0, nudge=<0 0>}
      padding=<ltrb 80 64 80 48>
     }
    }
    dialogSIze {
     height Override {enable=false, value=592}
     max Aspect Ratio {enable=false, value=0}
     max Desired Height {enable=true, value=800}
     max Desired Width {enable=false, value=0}
     min Aspect Ratio {enable=false, value=0}
     min Desired Height {enable=true, value=592}
     min Desired Width {enable=false, value=0}
     width Override {enable=true, value=1259}
    }
    titleImagePadding=<ltrb 10 0 15 0>
    titleProperties {
     auto Wrap Text=true
     justification=Left
     line Height Percentage=1
     mobile Font Size Multiplier=1.3
     transform Policy=ToUpper
     wrap Text At=700
     color {colorUseRule=UseColor_Specified, specifiedColor=<rgba 1 1 1 1>}
     font {
      bForceMonospaced=false
      bMaterialIsStencil=false
      fontMaterial=null
      fontObject=&HeadingNow
      letterSpacing=0
      monospacedWidth=1
      size=25.5
      skewAmount=0
      typefaceFontName=86 Bold Italic
      outlineSettings {
       bApplyOutlineToDropShadows=false
       bMiteredCorners=false
       bSeparateFillAlpha=false
       outlineMaterial=null
       outlineSize=0
       outlineColor=<rgba 0 0 0 1>
      }
     }
     placement {
      horizontal Alignment=HAlign_Fill
      vertical Alignment=VAlign_Top
      grid {column=1, column Span=0, layer=0, row=0, row Span=0, nudge=<0 0>}
      padding=<ltrb 80 80 80 32>
     }
    }
   }
  }
  GridPanel_1 {widgetClassPath=&GridPanel, namedSlotHost=Dialog_Internal, widget=GridPanel_1, properties {columnFill [1], rowFill [1]}}
  S_Content {
   widgetClassPath=&CommonVisibilitySwitcher
   bIsVariable=true
   parent=GridPanel_1
   widget=S_Content
   properties {navigation=&self:WidgetTree.S_Content.WidgetNavigation_0}
   slot {class=GridSlot, properties {layer=1, padding=<ltrb 80 80 80 48>}}
  }
  GP_SubmitReport {
   widgetClassPath=&GridPanel
   bIsVariable=true
   parent=S_Content
   widget=GP_SubmitReport
   properties {columnFill [1], rowFill [0,1,0]}
   slot {class=CommonVisibilitySwitcherSlot}
  }
  ReportChat_Title {
   widgetClassPath=&CommonTextBlock
   bIsVariable=true
   parent=GP_SubmitReport
   widget=ReportChat_Title
   properties {
    autoWrapText=true
    bIsScrollingEnabled=false
    style=&TextStyle-Base_C
    text=REPORT CHAT
    font {
     bForceMonospaced=false
     bMaterialIsStencil=false
     fontMaterial=null
     fontObject=&NotoSans
     letterSpacing=0
     monospacedWidth=1
     size=24
     skewAmount=0
     typefaceFontName=NotoSans-Regular
     outlineSettings {
      bApplyOutlineToDropShadows=false
      bMiteredCorners=false
      bSeparateFillAlpha=false
      outlineMaterial=null
      outlineSize=0
      outlineColor=<rgba 0 0 0 1>
     }
    }
    shadowOffset=<0 0>
   }
   slot {class=GridSlot}
  }
  S_ReportContent {
   widgetClassPath=&CommonVisibilitySwitcher
   bIsVariable=true
   parent=GP_SubmitReport
   widget=S_ReportContent
   slot {class=GridSlot, properties {row=1, padding=<ltrb 0 32 0 0>}}
  }
  GP_Report {
   widgetClassPath=&GridPanel
   bIsVariable=true
   parent=S_ReportContent
   widget=GP_Report
   properties {columnFill [1], rowFill [1]}
   slot {class=CommonVisibilitySwitcherSlot}
  }
  VerticalBox_0 {widgetClassPath=&VerticalBox, parent=GP_Report, widget=VerticalBox_0, slot {class=GridSlot}}
  ReportChat_Subtitle {
   widgetClassPath=&CommonTextBlock
   bIsVariable=true
   parent=VerticalBox_0
   widget=ReportChat_Subtitle
   properties {
    autoWrapText=true
    bIsScrollingEnabled=false
    text=Name x Members
    colorAndOpacity {colorUseRule=UseColor_Specified, specifiedColor=<rgba 0.5776 0.5776 0.5776 1>}
    font {
     bForceMonospaced=false
     bMaterialIsStencil=false
     fontMaterial=null
     fontObject=&HeadingNow
     letterSpacing=10
     monospacedWidth=1
     size=15
     skewAmount=0
     typefaceFontName=74 Regular
     outlineSettings {
      bApplyOutlineToDropShadows=false
      bMiteredCorners=false
      bSeparateFillAlpha=false
      outlineMaterial=null
      outlineSize=0
      outlineColor=<rgba 0 0 0 1>
     }
    }
    shadowOffset=<0 0>
   }
   slot {class=VerticalBoxSlot, properties {padding=<ltrb 0 0 0 12>}}
  }
  ReportChat_DescriptionText {
   widgetClassPath=&CommonRichTextBlock
   bIsVariable=true
   parent=VerticalBox_0
   widget=ReportChat_DescriptionText
   properties {
    autoWrapText=true
    bIsScrollingEnabled=false
    bOverrideDefaultStyle=true
    text="Your report will submit all the reportable text chat messages in your <img id=\"Party\"></> <Bold>{CurrentChannel}</> channel to Epic Games for review."
    textStyleSet=/FortniteCoreUI/UI/Social/Chat/TextChat/RichStyles/RS_UI_TextChat_Dialog
    decoratorClasses [/FortniteCoreUI/UI/Social/Chat/TextChat/RichTextDecorator/RichText_TextChatDecorator.RichText_TextChatDecorator_C]
    defaultTextStyleOverride {
     overflowPolicy=Clip
     transformPolicy=None
     colorAndOpacity {colorUseRule=UseColor_Specified, specifiedColor=<rgba 0.5776 0.5776 0.5776 1>}
     font {
      bForceMonospaced=false
      bMaterialIsStencil=false
      fontMaterial=null
      fontObject=&HeadingNow
      letterSpacing=10
      monospacedWidth=1
      size=18
      skewAmount=0
      typefaceFontName=74 Regular
      outlineSettings {
       bApplyOutlineToDropShadows=false
       bMiteredCorners=false
       bSeparateFillAlpha=false
       outlineMaterial=null
       outlineSize=0
       outlineColor=<rgba 0 0 0 1>
      }
     }
     highlightColor {colorUseRule=UseColor_Specified, specifiedColor=<rgba 0 0 0 1>}
     highlightShape {
      bIsDynamicallyLoaded=false
      drawAs=Image
      imageType=NoImage
      mirroring=NoMirror
      resourceName=None
      resourceObject=null
      tiling=NoTile
      imageSize=<32 32>
      margin=<ltrb 0 0 0 0>
      outlineSettings {
       bUseBrushTransparency=false
       roundingType=HalfHeightRadius
       width=0
       color {colorUseRule=UseColor_Specified, specifiedColor=<rgba 0 0 0 0>}
       cornerRadii=<0 0 0 0>
      }
      tintColor {colorUseRule=UseColor_Specified, specifiedColor=<rgba 1 1 1 1>}
      uVRegion {bIsValid=false, max=<0 0>, min=<0 0>}
     }
     selectedBackgroundColor {colorUseRule=UseColor_Foreground, specifiedColor=<rgba 1 0 1 1>}
     shadowColorAndOpacity=<rgba 0 0 0 1>
     shadowOffset=<0 0>
     strikeBrush {
      bIsDynamicallyLoaded=false
      drawAs=Image
      imageType=NoImage
      mirroring=NoMirror
      resourceName=None
      resourceObject=null
      tiling=NoTile
      imageSize=<32 32>
      margin=<ltrb 0 0 0 0>
      outlineSettings {
       bUseBrushTransparency=false
       roundingType=HalfHeightRadius
       width=0
       color {colorUseRule=UseColor_Specified, specifiedColor=<rgba 0 0 0 0>}
       cornerRadii=<0 0 0 0>
      }
      tintColor {colorUseRule=UseColor_Specified, specifiedColor=<rgba 1 1 1 1>}
      uVRegion {bIsValid=false, max=<0 0>, min=<0 0>}
     }
     underlineBrush {
      bIsDynamicallyLoaded=false
      drawAs=Image
      imageType=NoImage
      mirroring=NoMirror
      resourceName=None
      resourceObject=null
      tiling=NoTile
      imageSize=<32 32>
      margin=<ltrb 0 0 0 0>
      outlineSettings {
       bUseBrushTransparency=false
       roundingType=HalfHeightRadius
       width=0
       color {colorUseRule=UseColor_Specified, specifiedColor=<rgba 0 0 0 0>}
       cornerRadii=<0 0 0 0>
      }
      tintColor {colorUseRule=UseColor_Specified, specifiedColor=<rgba 1 1 1 1>}
      uVRegion {bIsValid=false, max=<0 0>, min=<0 0>}
     }
    }
   }
   slot {class=VerticalBoxSlot}
  }
  GP_ReportButtons {
   widgetClassPath=&GridPanel
   parent=GP_Report
   widget=GP_ReportButtons
   properties {columnFill [1], rowFill [1]}
   slot {class=GridSlot, properties {horizontalAlignment=HAlign_Left, row=1, padding=<ltrb 0 64 0 0>}}
  }
  Button_SubmitReport {
   widgetClassPath=/FortUILibrary/UIKit/Buttons/WBP_UIKit_Button_Loud
   bIsVariable=true
   parent=GP_ReportButtons
   widget=Button_SubmitReport
   properties {
    bDisplayInputActionWhenNotInteractable=false
    bShouldUseFallbackDefaultInputAction=false
    navigation=&self:WidgetTree.Button_SubmitReport.WidgetNavigation_0
    text=Submit
    visibility=Visible
   }
   slot {class=GridSlot}
  }
  Button_Nevermind {
   widgetClassPath=&WBP_UIKit_Button_Regular
   bIsVariable=true
   parent=GP_ReportButtons
   widget=Button_Nevermind
   properties {
    bDisplayInputActionWhenNotInteractable=false
    bShouldUseFallbackDefaultInputAction=false
    buttonType=Quiet
    navigation=&self:WidgetTree.Button_Nevermind.WidgetNavigation_0
    text=Cancel
    triggeringInputAction {dataTable=&InputActionDataTable, rowName=GenericBack}
   }
   slot {class=GridSlot, properties {column=1, padding=<ltrb 24 0 0 0>}}
  }
  GP_Loading {
   widgetClassPath=&GridPanel
   bIsVariable=true
   parent=S_ReportContent
   widget=GP_Loading
   properties {columnFill [1], rowFill [1]}
   slot {class=CommonVisibilitySwitcherSlot}
  }
  VerticalBox_1 {
   widgetClassPath=&VerticalBox
   parent=GP_Loading
   widget=VerticalBox_1
   slot {class=GridSlot, properties {horizontalAlignment=HAlign_Center, verticalAlignment=VAlign_Center}}
  }
  CommonTextBlock_63 {
   widgetClassPath=&CommonTextBlock
   parent=VerticalBox_1
   widget=CommonTextBlock_63
   properties {
    bIsScrollingEnabled=false
    style=&TextStyle-Base_C
    text=Sending Report
    font {
     bForceMonospaced=false
     bMaterialIsStencil=false
     fontMaterial=null
     fontObject=&NotoSans
     letterSpacing=0
     monospacedWidth=1
     size=24
     skewAmount=0
     typefaceFontName=NotoSans-Regular
     outlineSettings {
      bApplyOutlineToDropShadows=false
      bMiteredCorners=false
      bSeparateFillAlpha=false
      outlineMaterial=null
      outlineSize=0
      outlineColor=<rgba 0 0 0 1>
     }
    }
    shadowOffset=<0 0>
   }
   slot {class=VerticalBoxSlot}
  }
  LoadingSpinner {
   widgetClassPath=/FortUILibrary/UIKit/Throbber/WBP_UIKit_Throbber
   parent=VerticalBox_1
   widget=LoadingSpinner
   properties {displayBackground=false, throbberSize=Large}
   slot {class=VerticalBoxSlot}
  }
  GP_Error {
   widgetClassPath=&GridPanel
   bIsVariable=true
   parent=S_ReportContent
   widget=GP_Error
   properties {columnFill [1], rowFill [1]}
   slot {class=CommonVisibilitySwitcherSlot}
  }
  VBMessage {
   widgetClassPath=&VerticalBox
   parent=GP_Error
   widget=VBMessage
   slot {class=GridSlot, properties {horizontalAlignment=HAlign_Center, verticalAlignment=VAlign_Center}}
  }
  Text_ErrorHeader {
   widgetClassPath=&CommonTextBlock
   bIsVariable=true
   parent=VBMessage
   widget=Text_ErrorHeader
   properties {
    justification=Center
    text=No Results
    textTransformPolicy=ToUpper
    font {
     bForceMonospaced=false
     bMaterialIsStencil=false
     fontMaterial=null
     fontObject=&Burbank
     letterSpacing=0
     monospacedWidth=1
     size=32
     skewAmount=0.2
     typefaceFontName=BurbankRegular-Black
     outlineSettings {
      bApplyOutlineToDropShadows=false
      bMiteredCorners=false
      bSeparateFillAlpha=false
      outlineMaterial=null
      outlineSize=0
      outlineColor=<rgba 0 0 0 1>
     }
    }
    shadowOffset=<0 0>
   }
   slot {class=VerticalBoxSlot, properties {horizontalAlignment=HAlign_Center, padding=<ltrb 0 0 0 8>}}
  }
  Text_ErrorMessage {
   widgetClassPath=&CommonTextBlock
   bIsVariable=true
   parent=VBMessage
   widget=Text_ErrorMessage
   properties {
    justification=Center
    renderOpacity=0.7
    text=To report messages ensure that text reporting is set to Always On in your settings and that there are reportable messages in your text chat.
    textTransformPolicy=ToUpper
    wrapTextAt=800
    font {
     bForceMonospaced=false
     bMaterialIsStencil=false
     fontMaterial=null
     fontObject=&Burbank
     letterSpacing=0
     monospacedWidth=1
     size=32
     skewAmount=0.2
     typefaceFontName=BurbankRegular-Black
     outlineSettings {
      bApplyOutlineToDropShadows=false
      bMiteredCorners=false
      bSeparateFillAlpha=false
      outlineMaterial=null
      outlineSize=0
      outlineColor=<rgba 0 0 0 1>
     }
    }
    shadowOffset=<0 0>
   }
   slot {class=VerticalBoxSlot, properties {horizontalAlignment=HAlign_Center}}
  }
  Button_CloseError {
   widgetClassPath=&WBP_UIKit_Button_Loud
   bIsVariable=true
   parent=GP_Error
   widget=Button_CloseError
   properties {text=Close, visibility=Visible, triggeringInputAction {dataTable=&InputActionDataTable, rowName=GenericBack}}
   slot {class=GridSlot, properties {horizontalAlignment=HAlign_Left, verticalAlignment=VAlign_Bottom}}
  }
  GP_ReportSubmitted {
   widgetClassPath=&GridPanel
   bIsVariable=true
   parent=S_Content
   widget=GP_ReportSubmitted
   properties {columnFill [1], rowFill [0,1,0]}
   slot {class=CommonVisibilitySwitcherSlot}
  }
  HorizontalBox_0 {
   widgetClassPath=&HorizontalBox
   parent=GP_ReportSubmitted
   widget=HorizontalBox_0
   slot {class=GridSlot, properties {horizontalAlignment=HAlign_Left, verticalAlignment=VAlign_Top}}
  }
  Image_62 {
   widgetClassPath=&Image
   parent=HorizontalBox_0
   widget=Image_62
   properties {brush {
     bIsDynamicallyLoaded=false
     drawAs=Image
     imageType=FullColor
     mirroring=NoMirror
     resourceName=None
     resourceObject=/Game/Athena/UI/PostMatch/AthenaChallenge/Textures/T_UI_CheckMark
     tiling=NoTile
     imageSize=<32 32>
     margin=<ltrb 0 0 0 0>
     outlineSettings {
      bUseBrushTransparency=false
      roundingType=HalfHeightRadius
      width=0
      color {colorUseRule=UseColor_Specified, specifiedColor=<rgba 0 0 0 0>}
      cornerRadii=<0 0 0 0>
     }
     tintColor {colorUseRule=UseColor_Specified, specifiedColor=<rgba 1 1 1 1>}
     uVRegion {bIsValid=false, max=<0 0>, min=<0 0>}
    }}
   slot {class=HorizontalBoxSlot, properties {horizontalAlignment=HAlign_Left, verticalAlignment=VAlign_Center, padding=<ltrb 0 0 16 0>}}
  }
  TitleText_1 {
   widgetClassPath=&CommonTextBlock
   parent=HorizontalBox_0
   widget=TitleText_1
   properties {
    bIsScrollingEnabled=false
    style=&TextStyle-Base_C
    text=Report submitted
    font {
     bForceMonospaced=false
     bMaterialIsStencil=false
     fontMaterial=null
     fontObject=&NotoSans
     letterSpacing=0
     monospacedWidth=1
     size=24
     skewAmount=0
     typefaceFontName=NotoSans-Regular
     outlineSettings {
      bApplyOutlineToDropShadows=false
      bMiteredCorners=false
      bSeparateFillAlpha=false
      outlineMaterial=null
      outlineSize=0
      outlineColor=<rgba 0 0 0 1>
     }
    }
    shadowOffset=<0 0>
   }
   slot {class=HorizontalBoxSlot, properties {horizontalAlignment=HAlign_Left}}
  }
  Text_ReportSubmitted {
   widgetClassPath=&CommonRichTextBlock
   bIsVariable=true
   parent=GP_ReportSubmitted
   widget=Text_ReportSubmitted
   properties {
    autoWrapText=true
    bOverrideDefaultStyle=true
    text="Thank you for your report. To track the status of your report and any enforcement actions taken, check your My Reports page at <URL>https://epic.gm/my-reports/</> (ID: 999-999-999)\r\n\r\nReports help us to ensure that Fortnite remains a fair, safe and enjoyable game for all players."
    textStyleSet=&RS_UI_TextChat_Dialog
    defaultTextStyleOverride {
     overflowPolicy=Clip
     transformPolicy=None
     colorAndOpacity {colorUseRule=UseColor_Specified, specifiedColor=<rgba 1 1 1 1>}
     font {
      bForceMonospaced=false
      bMaterialIsStencil=false
      fontMaterial=null
      fontObject=&HeadingNow
      letterSpacing=10
      monospacedWidth=1
      size=18
      skewAmount=0
      typefaceFontName=74 Regular
      outlineSettings {
       bApplyOutlineToDropShadows=false
       bMiteredCorners=false
       bSeparateFillAlpha=false
       outlineMaterial=null
       outlineSize=0
       outlineColor=<rgba 0 0 0 1>
      }
     }
     highlightColor {colorUseRule=UseColor_Specified, specifiedColor=<rgba 0 0 0 1>}
     highlightShape {
      bIsDynamicallyLoaded=false
      drawAs=Image
      imageType=NoImage
      mirroring=NoMirror
      resourceName=None
      resourceObject=null
      tiling=NoTile
      imageSize=<32 32>
      margin=<ltrb 0 0 0 0>
      outlineSettings {
       bUseBrushTransparency=false
       roundingType=HalfHeightRadius
       width=0
       color {colorUseRule=UseColor_Specified, specifiedColor=<rgba 0 0 0 0>}
       cornerRadii=<0 0 0 0>
      }
      tintColor {colorUseRule=UseColor_Specified, specifiedColor=<rgba 1 1 1 1>}
      uVRegion {bIsValid=false, max=<0 0>, min=<0 0>}
     }
     selectedBackgroundColor {colorUseRule=UseColor_Foreground, specifiedColor=<rgba 1 0 1 1>}
     shadowColorAndOpacity=<rgba 0 0 0 1>
     shadowOffset=<0 0>
     strikeBrush {
      bIsDynamicallyLoaded=false
      drawAs=Image
      imageType=NoImage
      mirroring=NoMirror
      resourceName=None
      resourceObject=null
      tiling=NoTile
      imageSize=<32 32>
      margin=<ltrb 0 0 0 0>
      outlineSettings {
       bUseBrushTransparency=false
       roundingType=HalfHeightRadius
       width=0
       color {colorUseRule=UseColor_Specified, specifiedColor=<rgba 0 0 0 0>}
       cornerRadii=<0 0 0 0>
      }
      tintColor {colorUseRule=UseColor_Specified, specifiedColor=<rgba 1 1 1 1>}
      uVRegion {bIsValid=false, max=<0 0>, min=<0 0>}
     }
     underlineBrush {
      bIsDynamicallyLoaded=false
      drawAs=Image
      imageType=NoImage
      mirroring=NoMirror
      resourceName=None
      resourceObject=null
      tiling=NoTile
      imageSize=<32 32>
      margin=<ltrb 0 0 0 0>
      outlineSettings {
       bUseBrushTransparency=false
       roundingType=HalfHeightRadius
       width=0
       color {colorUseRule=UseColor_Specified, specifiedColor=<rgba 0 0 0 0>}
       cornerRadii=<0 0 0 0>
      }
      tintColor {colorUseRule=UseColor_Specified, specifiedColor=<rgba 1 1 1 1>}
      uVRegion {bIsValid=false, max=<0 0>, min=<0 0>}
     }
    }
   }
   slot {class=GridSlot, properties {row=1, padding=<ltrb 0 32 0 64>}}
  }
  GP_SubmittedButtons {
   widgetClassPath=&GridPanel
   parent=GP_ReportSubmitted
   widget=GP_SubmittedButtons
   properties {columnFill [1], rowFill [1]}
   slot {class=GridSlot, properties {horizontalAlignment=HAlign_Left, row=2}}
  }
  Button_Close {
   widgetClassPath=&WBP_UIKit_Button_Loud
   bIsVariable=true
   parent=GP_SubmittedButtons
   widget=Button_Close
   properties {
    bDisplayInputActionWhenNotInteractable=false
    bShouldUseFallbackDefaultInputAction=false
    navigation=&self:WidgetTree.Button_Close.WidgetNavigation_0
    text=Close
    visibility=Visible
    triggeringInputAction {dataTable=&InputActionDataTable, rowName=GenericBack}
   }
   slot {class=GridSlot}
  }
  SB_Mobile {
   widgetClassPath=&SizeBox
   bIsVariable=true
   parent=GP_SubmittedButtons
   widget=SB_Mobile
   slot {class=GridSlot, properties {column=1, padding=<ltrb 24 0 0 0>}}
  }
  Button_PlayerBlock {
   widgetClassPath=&WBP_UIKit_Button_Regular
   bIsVariable=true
   parent=SB_Mobile
   widget=Button_PlayerBlock
   properties {
    bDisplayInputActionWhenNotInteractable=false
    bShouldUseFallbackDefaultInputAction=false
    buttonType=Quiet
    navigation=&self:WidgetTree.Button_PlayerBlock.WidgetNavigation_0
    text=Choose Players To Block
   }
   slot {class=SizeBoxSlot}
  }
  GP_BlockPlayer {
   widgetClassPath=&GridPanel
   bIsVariable=true
   parent=S_Content
   widget=GP_BlockPlayer
   properties {columnFill [1], rowFill [0,0,1,0]}
   slot {class=CommonVisibilitySwitcherSlot}
  }
  TitleText_2 {
   widgetClassPath=&CommonTextBlock
   parent=GP_BlockPlayer
   widget=TitleText_2
   properties {
    bIsScrollingEnabled=false
    style=&TextStyle-Base_C
    text=CHOOSE PLAYERS TO BLOCK
    font {
     bForceMonospaced=false
     bMaterialIsStencil=false
     fontMaterial=null
     fontObject=&NotoSans
     letterSpacing=0
     monospacedWidth=1
     size=24
     skewAmount=0
     typefaceFontName=NotoSans-Regular
     outlineSettings {
      bApplyOutlineToDropShadows=false
      bMiteredCorners=false
      bSeparateFillAlpha=false
      outlineMaterial=null
      outlineSize=0
      outlineColor=<rgba 0 0 0 1>
     }
    }
    shadowOffset=<0 0>
   }
   slot {class=GridSlot, properties {horizontalAlignment=HAlign_Left}}
  }
  VerticalBox_2 {
   widgetClassPath=&VerticalBox
   parent=GP_BlockPlayer
   widget=VerticalBox_2
   slot {class=GridSlot, properties {row=1, padding=<ltrb 0 32 0 0>}}
  }
  BlockPlayer_Subtitle {
   widgetClassPath=&CommonTextBlock
   bIsVariable=true
   parent=VerticalBox_2
   widget=BlockPlayer_Subtitle
   properties {
    autoWrapText=true
    bIsScrollingEnabled=false
    text=Name x Members
    colorAndOpacity {colorUseRule=UseColor_Specified, specifiedColor=<rgba 0.5776 0.5776 0.5776 1>}
    font {
     bForceMonospaced=false
     bMaterialIsStencil=false
     fontMaterial=null
     fontObject=&HeadingNow
     letterSpacing=10
     monospacedWidth=1
     size=15
     skewAmount=0
     typefaceFontName=74 Regular
     outlineSettings {
      bApplyOutlineToDropShadows=false
      bMiteredCorners=false
      bSeparateFillAlpha=false
      outlineMaterial=null
      outlineSize=0
      outlineColor=<rgba 0 0 0 1>
     }
    }
    shadowOffset=<0 0>
   }
   slot {class=VerticalBoxSlot, properties {padding=<ltrb 0 0 0 12>}}
  }
  DescriptionText_2 {
   widgetClassPath=&CommonTextBlock
   parent=VerticalBox_2
   widget=DescriptionText_2
   properties {
    autoWrapText=true
    style=&TextStyle-Base_C
    text="Blocked players will be removed from your friends list, cannot invite you to their party, and cannot voice chat or text chat with you. Blocked players will not be notified that you blocked them."
    font {
     bForceMonospaced=false
     bMaterialIsStencil=false
     fontMaterial=null
     fontObject=&NotoSans
     letterSpacing=0
     monospacedWidth=1
     size=24
     skewAmount=0
     typefaceFontName=NotoSans-Regular
     outlineSettings {
      bApplyOutlineToDropShadows=false
      bMiteredCorners=false
      bSeparateFillAlpha=false
      outlineMaterial=null
      outlineSize=0
      outlineColor=<rgba 0 0 0 1>
     }
    }
    shadowOffset=<0 0>
   }
   slot {class=VerticalBoxSlot}
  }
  SizeBox_0 {
   widgetClassPath=&SizeBox
   parent=GP_BlockPlayer
   widget=SizeBox_0
   properties {
    bOverride_HeightOverride=true
    heightOverride=367
    navigation=&self:WidgetTree.SizeBox_0.WidgetNavigation_0
   }
   slot {class=GridSlot, properties {row=2, padding=<ltrb 0 0 0 64>}}
  }
  CommonTileView {
   widgetClassPath=/Script/CommonUI.CommonTileView
   bIsVariable=true
   parent=SizeBox_0
   widget=CommonTileView
   properties {
    bClearSelectionOnClick=true
    entryHeight=64
    entryWidgetClass=/FortniteCoreUI/UI/Social/Chat/TextChat/WBP_TextChat_ToggleButton.WBP_TextChat_ToggleButton_C
    entryWidth=460
    navigation=&self:WidgetTree.CommonTileView.WidgetNavigation_0
    numDesignerPreviewEntries=12
    selectionMode=Multi
    tileAlignment=Fill
    visibility=SelfHitTestInvisible
   }
   slot {class=SizeBoxSlot, properties {padding=<ltrb 0 31 0 16>}}
  }
  GP_BlockButtons {
   widgetClassPath=&GridPanel
   parent=GP_BlockPlayer
   widget=GP_BlockButtons
   properties {columnFill [1], rowFill [1]}
   slot {class=GridSlot, properties {horizontalAlignment=HAlign_Left, row=3}}
  }
  Button_Block {
   widgetClassPath=&WBP_UIKit_Button_Loud
   bIsVariable=true
   parent=GP_BlockButtons
   widget=Button_Block
   properties {
    navigation=&self:WidgetTree.Button_Block.WidgetNavigation_0
    text=Block
    triggeringInputAction {dataTable=&InputActionDataTable, rowName=ChatBlock}
   }
   slot {class=GridSlot}
  }
  Button_Close_Block {
   widgetClassPath=&WBP_UIKit_Button_Regular
   bIsVariable=true
   parent=GP_BlockButtons
   widget=Button_Close_Block
   properties {
    buttonType=Quiet
    navigation=&self:WidgetTree.Button_Close_Block.WidgetNavigation_0
    text=Close
    triggeringInputAction {dataTable=&InputActionDataTable, rowName=GenericBack}
   }
   slot {class=GridSlot, properties {column=1, padding=<ltrb 24 0 0 0>}}
  }
 ]
}
